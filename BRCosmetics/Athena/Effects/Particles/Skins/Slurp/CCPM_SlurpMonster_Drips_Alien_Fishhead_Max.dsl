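; Cosmetic character-part modifier Blueprint for the Alien Fishhead Max Slurp Monster skin, handling its slurp drip effects.
; When the pawn and part change, it runs the parent logic, then passes its slurp transformation effect and looping Niagara drips to SetComponentPointers.
; The construction script only calls the parent, so the remaining behavior is inherited.
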(event EventOnPawnAndPartChanged
  (|CallParentFunction)
  (CallFunction|SetComponentPointers (Variables|Default|GetP_Slurp_Transformation) (Variables|Default|GetNiagara_Looping_Drips)))

(fn ConstructionScript
  (|CallParentFunction))
