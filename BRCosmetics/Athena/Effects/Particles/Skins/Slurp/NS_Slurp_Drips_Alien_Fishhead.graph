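# Niagara cosmetic effect for slurp drips on the Alien Fishhead skin, built from two sprite emitters.
# Drip_Head spawns occasional drops on a hemisphere that swell, stretch and creep downward, then emit death events with world-space positions.
# Drip_Drop, with no spawner of its own shown, supplies short-lived, fast-falling, shrinking drops, presumably triggered by those events.
# Both lerp off-white to cyan via User.SlurpColorAlpha and modulate opacity by a per-particle DistanceAlpha.
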NiagaraSystem /BRCosmetics/Athena/Effects/Particles/Skins/Slurp/NS_Slurp_Drips_Alien_Fishhead
dependencies {
 usedDataInterfaces [
  &NiagaraDataInterfaceVectorCurve
  &NiagaraDataInterfaceVector2DCurve
  &NiagaraDataInterfaceCurve
 ]
 usedDynamicInputs []
 usedModules [
  &SystemState
  &self:SystemSpawnScript.SystemScriptSource.SystemScriptGraph.NiagaraNodeAssignment_0.SetVariables_#1
 ]
}
emitters {
 Drip_Drop {
  emitterUpdateScript [&EmitterState(Life Cycle Mode=::NewEnumerator1, Loop Duration=2, UseLoopDelay=false)]
  particleSpawnScript [
   /Niagara/Modules/Spawn/Initialization/InitializeParticle(
    Lifetime=&UniformRangedFloat(Minimum=0.6, Maximum=0.8)
    Mass=&UniformRangedFloat(Minimum=0.75, Maximum=1.75)
    Module.Write Color=true
    Module.Write Lifetime=true
    Module.Write Mass=false
    Module.Write Material Random=true
    Module.Write Position=false
    Module.Write Scale=false
    Module.Write SpriteRotation=false
    Module.Write SpriteSize=true
    Position=&SimulationPosition()
    Sprite Size=&Vector2DFromFloat(Value=&MakeFloatFromVector(Vector=&UniformRangedVector(Minimum=<5.5 0 0>, Maximum=<6.5 1 1>), Channel=::NewEnumerator0))
   )
   /Niagara/Modules/Spawn/Location/SphereLocation(MaskSpawnGroup=false, Module.Override Randomness=false, Offset=<-10 0 0>, Sphere Radius=0, Surface Only=false)
   &InheritVelocity(Module.Limit Speed=false, Velocity Scale=<-0.3 -0.3 -0.3>)
  ]
  particleUpdateScript [
   &CameraOffset(Camera Offset Amount=12)
   &ParticleState()
   &AccelerationForce(Acceleration=&VectorFromCurve(VectorCurve=DataInterface() {propertyValues {
       XCurve {keys [<at 0 0>,<at 0.9989 0>]}
       YCurve {keys [<at 0 0>,<at 0.9989 0>]}
       ZCurve {keys [
         {arriveTangent=-15.5802, interpMode=RCIM_Cubic, leaveTangent=-8.2205, tangentMode=RCTM_User, time=0, value=0}
         {arriveTangent=-134.1077, interpMode=RCIM_Cubic, leaveTangent=-221.0645, tangentMode=RCTM_User, time=0.3, value=-600}
        ]}
      }}))
   &SolveForcesAndVelocity(Clamp Velocity=false, Limit Acceleration=false)
   /SetVariables() {isSetParameters=true}
   &ScaleColor(
    Scale Alpha=&Lerp_Float(A=-1, B=1, Alpha=$Particles.DistanceAlpha)
    Scale RGB=&MakeVectorFromLinearColorRGB(Color=&LerpLinearColors(StartColor=<rgba 0.9 0.9 0.9 1>, EndColor=<rgba 0.1396 1.4602 1.5 1>, LerpFactor=$User.SlurpColorAlpha))
    ScaleA=true
    ScaleRGB=true
   )
   &ScaleSpriteSize(Scale Factor=&Vector2DFromCurve(
     Vector2Curve=DataInterface() {propertyValues {
       XCurve {keys [
         {arriveTangent=5.9728, interpMode=RCIM_Cubic, leaveTangent=5.9728, tangentMode=RCTM_User, time=0, value=0.1}
         {arriveTangent=-0.1, interpMode=RCIM_Cubic, leaveTangent=-0.1, tangentMode=RCTM_User, time=0.18, value=0.99}
         {interpMode=RCIM_Cubic, tangentMode=RCTM_User, time=1, value=0}
        ]}
       YCurve {keys [
         {arriveTangent=6.2715, interpMode=RCIM_Cubic, leaveTangent=6.2715, tangentMode=RCTM_User, time=0, value=0.1}
         {arriveTangent=-0.1001, interpMode=RCIM_Cubic, leaveTangent=-0.1001, tangentMode=RCTM_User, time=0.18, value=0.99}
         {interpMode=RCIM_Cubic, tangentMode=RCTM_User, time=0.9989, value=0}
        ]}
      }}
     ScaleCurve=<1 1>
    ))
  ]
  rendererClasses [&NiagaraSpriteRendererProperties]
  renderers [{rendererClass=&NiagaraSpriteRendererProperties}]
 }
 Drip_Head {
  emitterUpdateScript [
   &EmitterState(Life Cycle Mode=::NewEnumerator1, Loop Duration=4, UseLoopDelay=false)
   &SpawnBurst_Instantaneous(Spawn Count=1, Spawn Time=&UniformRangedFloat(Minimum=1, Maximum=4), Use Spawn Probability=false) {enabled=false}
   &SpawnRate(SpawnRate=0.3, Use Spawn Probability=false)
  ]
  particleSpawnScript [
   &InitializeParticle(
    Lifetime=&UniformRangedFloat(Minimum=1.8, Maximum=2.1)
    Mass=&UniformRangedFloat(Minimum=0.75, Maximum=1.75)
    Module.Write Color=true
    Module.Write Lifetime=true
    Module.Write Mass=false
    Module.Write Material Random=true
    Module.Write Position=false
    Module.Write Scale=false
    Module.Write SpriteRotation=true
    Module.Write SpriteSize=true
    Position=&SimulationPosition()
    Sprite Rotation=180
    Sprite Size=&Vector2DFromFloat(Value=&MakeFloatFromVector(Vector=&UniformRangedVector(Minimum=<3 0 0>, Maximum=<4 0 0>), Channel=::NewEnumerator0))
   )
   &SphereLocation(Hemisphere X=true, MaskSpawnGroup=false, Module.Override Randomness=false, Offset=<11 3 0>, Sphere Radius=13.5, Surface Only=true)
  ]
  particleUpdateScript [
   &CameraOffset(Camera Offset Amount=10)
   &ParticleState()
   &AccelerationForce(Acceleration=&VectorFromCurve(VectorCurve=DataInterface() {propertyValues {
       XCurve {keys [<at 0 0>,<at 0.9989 0>]}
       YCurve {keys [<at 0 0>,<at 0.9989 0>]}
       ZCurve {keys [
         {arriveTangent=-15.5802, interpMode=RCIM_Cubic, leaveTangent=-8.2205, tangentMode=RCTM_User, time=0, value=-1}
         {arriveTangent=-134.1077, interpMode=RCIM_Cubic, leaveTangent=-221.0645, tangentMode=RCTM_User, time=0.3, value=-10}
        ]}
      }}))
   &SolveForcesAndVelocity(Clamp Velocity=false, Limit Acceleration=false)
   /SetVariables() {isSetParameters=true}
   &ScaleSpriteSize(Scale Factor=&Vector2DFromCurve(
     Vector2Curve=DataInterface() {propertyValues {
       XCurve {keys [
         {interpMode=RCIM_Cubic, tangentMode=RCTM_User, time=0, value=1.15}
         {arriveTangent=-0.8621, interpMode=RCIM_Cubic, leaveTangent=-0.8621, tangentMode=RCTM_User, time=0.6, value=1.14}
         <at 0.8265 0.457>
         {arriveTangent=-0.351, interpMode=RCIM_Cubic, leaveTangent=-0.351, tangentMode=RCTM_User, time=0.87, value=0.4}
         {interpMode=RCIM_Cubic, tangentMode=RCTM_User, time=0.9988, value=1}
        ]}
       YCurve {keys [
         {interpMode=RCIM_Cubic, tangentMode=RCTM_User, time=0, value=0}
         {arriveTangent=2.8736, interpMode=RCIM_Cubic, leaveTangent=2.8736, tangentMode=RCTM_User, time=0.6, value=0.99}
         {arriveTangent=2.025, interpMode=RCIM_Cubic, leaveTangent=2.025, tangentMode=RCTM_User, time=0.87, value=2.5}
         {interpMode=RCIM_Cubic, tangentMode=RCTM_User, time=1, value=1.8}
        ]}
      }}
     ScaleCurve=<1 1>
    ))
   &ScaleColor(
    Scale Alpha=&Lerp_Float(
     A=-1
     B=&FloatFromCurve(FloatCurve=DataInterface() {propertyValues {Curve {keys [<at 0 0>,<at 0.1 1>,<at 0.9 1>,<at 1 0>]}}})
     Alpha=$Particles.DistanceAlpha
    )
    Scale RGB=&MakeVectorFromLinearColorRGB(Color=&LerpLinearColors(StartColor=<rgba 0.9 0.9 0.9 1>, EndColor=<rgba 0.1396 1.4602 1.5 1>, LerpFactor=$User.SlurpColorAlpha))
    ScaleA=true
    ScaleRGB=true
   )
   &GenerateDeathEvent(
    Gap Correction Amount=1.5
    Position To Send:Vector3f=/Niagara/DynamicInputs/Transforms/TransformPosition(Position=Linked() {linkedVariable {name=Particles.Position, type=NiagaraPosition}}, Source Space=::Local, Destination Space=::World)
   )
  ]
  rendererClasses [&NiagaraSpriteRendererProperties]
  renderers [{rendererClass=&NiagaraSpriteRendererProperties}]
 }
}
summary {
 systemName=NS_Slurp_Drips_Alien_Fishhead
 emitters [
  {emitterName=Drip_Head, rendererClasses [&NiagaraSpriteRendererProperties]}
  {emitterName=Drip_Drop, rendererClasses [&NiagaraSpriteRendererProperties]}
 ]
 userVariables [User.SlurpColorAlpha {type=NiagaraFloat, defaultValue=0, description=""}]
}
