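# Niagara system for slime drips from the bandolier on the Slurp character skin, built as two pairs of sprite emitters.
# The A emitters randomly burst single drips that follow sampled bones 0 and 1 of User.SkeletalMesh, stretch as they age, and generate death events.
# The B emitters, presumably spawned by those events (handlers not shown), inherit velocity and fall as narrowing, elongating drops.
# User.SlurpColorAlpha blends all four from white to bright cyan.
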NiagaraSystem /BRCosmetics/Athena/Effects/Particles/Skins/Slurp/NS_Slurp_Drips_Bandolier
dependencies {
 usedDataInterfaces [
  &NiagaraDataInterfaceVectorCurve
  &NiagaraDataInterfaceVector2DCurve
  &NiagaraDataInterfaceCurve
 ]
 usedDynamicInputs []
 usedModules [&SystemState]
}
emitters {
 Drip_Chest_A {
  emitterUpdateScript [
   &EmitterState(
    Life Cycle Mode=::NewEnumerator1
    Loop Delay=&UniformRangedFloat(Minimum=0.8, Maximum=1.5)
    Loop Duration=1
    UseLoopDelay=true
   )
   &SpawnBurst_Instantaneous(Spawn Count=1, Spawn Probability=0.25, Spawn Time=0.333, Use Spawn Probability=true)
  ]
  particleSpawnScript [/Niagara/Modules/Spawn/Initialization/InitializeParticle(
    Lifetime=&UniformRangedFloat(Minimum=1.8, Maximum=2.1)
    Mass=&UniformRangedFloat(Minimum=0.75, Maximum=1.75)
    Module.Write Color=true
    Module.Write Lifetime=true
    Module.Write Mass=false
    Module.Write Material Random=true
    Module.Write Position=false
    Module.Write Scale=false
    Module.Write SpriteRotation=true
    Module.Write SpriteSize=true
    Position=&SimulationPosition()
    Sprite Rotation=180
    Sprite Size=&Vector2DFromFloat(Value=&MakeFloatFromVector(Vector=&UniformRangedVector(Minimum=<5 0 0>, Maximum=<6 1 1>), Channel=::NewEnumerator0))
   )]
  particleUpdateScript [
   &SampleSkeletalMeshSkeleton(
    Mesh:NiagaraDataInterfaceSkeletalMesh=$User.SkeletalMesh
    Specific Bone Index:NiagaraInt32=&MakeIntFromFloat(Float=0)
   )
   /Niagara/Modules/Spawn/Location/SkeletalMeshSkeletonLocation(Module.Inherit Velocity=false, Module.Use Sampled Bone as Ribbon ID=true, Sampled Position Offset:Vector3f=<-0.2 0 0>)
   &CameraOffset(Camera Offset Amount=2)
   &ParticleState()
   &AccelerationForce(Acceleration=&VectorFromCurve(VectorCurve=DataInterface() {propertyValues {
       XCurve {keys [<at 0 0>,<at 0.9989 0>]}
       YCurve {keys [<at 0 0>,<at 0.9989 0>]}
       ZCurve {keys [
         {arriveTangent=-15.5802, interpMode=RCIM_Cubic, leaveTangent=-8.2205, tangentMode=RCTM_User, time=0, value=-1}
         {arriveTangent=-134.1077, interpMode=RCIM_Cubic, leaveTangent=-221.0645, tangentMode=RCTM_User, time=0.3, value=-20}
        ]}
      }}))
   &SolveForcesAndVelocity(Clamp Velocity=false, Limit Acceleration=false)
   &ScaleSpriteSize(Scale Factor=&Vector2DFromCurve(
     Vector2Curve=DataInterface() {propertyValues {
       XCurve {keys [
         {interpMode=RCIM_Cubic, tangentMode=RCTM_User, time=0, value=1.15}
         {arriveTangent=-0.8621, interpMode=RCIM_Cubic, leaveTangent=-0.8621, tangentMode=RCTM_User, time=0.6, value=1.14}
         <at 0.8189 0.5563>
         {arriveTangent=-0.351, interpMode=RCIM_Cubic, leaveTangent=-0.351, tangentMode=RCTM_User, time=0.8716, value=0.5013}
         {interpMode=RCIM_Cubic, tangentMode=RCTM_User, time=0.9938, value=0.5013}
        ]}
       YCurve {keys [
         {interpMode=RCIM_Cubic, tangentMode=RCTM_User, time=0, value=0}
         {arriveTangent=2.8736, interpMode=RCIM_Cubic, leaveTangent=2.8736, tangentMode=RCTM_User, time=0.6, value=0.99}
         {arriveTangent=2.025, interpMode=RCIM_Cubic, leaveTangent=2.025, tangentMode=RCTM_User, time=0.87, value=2}
         {interpMode=RCIM_Cubic, tangentMode=RCTM_User, time=1, value=2.1}
        ]}
      }}
     ScaleCurve=<1 1>
    ))
   &ScaleColor(
    Scale Alpha=&FloatFromCurve(FloatCurve=DataInterface() {propertyValues {Curve {keys [<at 0 0>,<at 0.1 1>,<at 0.9 1>,<at 1 1>]}}})
    Scale RGB=&MakeVectorFromLinearColorRGB(Color=&LerpLinearColors(StartColor=<rgba 1 1 1 1>, EndColor=<rgba 0.1861 1.9469 2 1>, LerpFactor=$User.SlurpColorAlpha))
    ScaleA=true
    ScaleRGB=true
   )
   &GenerateDeathEvent(
    Gap Correction Amount=1.5
    Position To Send:Vector3f=/Niagara/DynamicInputs/Transforms/TransformPosition(Position=Linked() {linkedVariable {name=Particles.Position, type=NiagaraPosition}}, Source Space=::Local, Destination Space=::World)
   )
  ]
  rendererClasses [&NiagaraSpriteRendererProperties]
  renderers [{rendererClass=&NiagaraSpriteRendererProperties}]
 }
 Drip_Chest_A001 {
  emitterUpdateScript [
   &EmitterState(
    Life Cycle Mode=::NewEnumerator1
    Loop Delay=&UniformRangedFloat(Minimum=0.8, Maximum=1.5)
    Loop Duration=1
    UseLoopDelay=true
   )
   &SpawnBurst_Instantaneous(Spawn Count=1, Spawn Probability=0.25, Use Spawn Probability=true)
  ]
  particleSpawnScript [&InitializeParticle(
    Lifetime=&UniformRangedFloat(Minimum=1.8, Maximum=2.1)
    Mass=&UniformRangedFloat(Minimum=0.75, Maximum=1.75)
    Module.Write Color=true
    Module.Write Lifetime=true
    Module.Write Mass=false
    Module.Write Material Random=true
    Module.Write Position=false
    Module.Write Scale=false
    Module.Write SpriteRotation=true
    Module.Write SpriteSize=true
    Position=&SimulationPosition()
    Sprite Rotation=180
    Sprite Size=&Vector2DFromFloat(Value=&MakeFloatFromVector(Vector=&UniformRangedVector(Minimum=<5.5 0 0>, Maximum=<6.5 1 1>), Channel=::NewEnumerator0))
   )]
  particleUpdateScript [
   &SampleSkeletalMeshSkeleton(
    Mesh:NiagaraDataInterfaceSkeletalMesh=$User.SkeletalMesh
    Specific Bone Index:NiagaraInt32=&MakeIntFromFloat(Float=1)
   )
   &SkeletalMeshSkeletonLocation(Module.Inherit Velocity=false, Module.Use Sampled Bone as Ribbon ID=true, Sampled Position Offset:Vector3f=<0 0 1.5>)
   &CameraOffset(Camera Offset Amount=2)
   &ParticleState()
   &AccelerationForce(Acceleration=&VectorFromCurve(VectorCurve=DataInterface() {propertyValues {
       XCurve {keys [<at 0 0>,<at 0.9989 0>]}
       YCurve {keys [<at 0 0>,<at 0.9989 0>]}
       ZCurve {keys [
         {arriveTangent=-15.5802, interpMode=RCIM_Cubic, leaveTangent=-8.2205, tangentMode=RCTM_User, time=0, value=-1}
         {arriveTangent=-134.1077, interpMode=RCIM_Cubic, leaveTangent=-221.0645, tangentMode=RCTM_User, time=0.3, value=-20}
        ]}
      }}))
   &SolveForcesAndVelocity(Clamp Velocity=false, Limit Acceleration=false)
   &ScaleSpriteSize(Scale Factor=&Vector2DFromCurve(
     Vector2Curve=DataInterface() {propertyValues {
       XCurve {keys [
         {interpMode=RCIM_Cubic, tangentMode=RCTM_User, time=0, value=1.15}
         {arriveTangent=-0.8621, interpMode=RCIM_Cubic, leaveTangent=-0.8621, tangentMode=RCTM_User, time=0.6, value=1.14}
         <at 0.818 0.5563>
         {arriveTangent=-0.351, interpMode=RCIM_Cubic, leaveTangent=-0.351, tangentMode=RCTM_User, time=0.87, value=0.5013}
         {interpMode=RCIM_Cubic, tangentMode=RCTM_User, time=0.9966, value=0.5013}
        ]}
       YCurve {keys [
         {interpMode=RCIM_Cubic, tangentMode=RCTM_User, time=0, value=0}
         {arriveTangent=2.8736, interpMode=RCIM_Cubic, leaveTangent=2.8736, tangentMode=RCTM_User, time=0.6, value=0.99}
         {arriveTangent=2.025, interpMode=RCIM_Cubic, leaveTangent=2.025, tangentMode=RCTM_User, time=0.87, value=2}
         {interpMode=RCIM_Cubic, tangentMode=RCTM_User, time=1, value=2.1}
        ]}
      }}
     ScaleCurve=<1 1>
    ))
   &ScaleColor(
    Scale Alpha=&FloatFromCurve(FloatCurve=DataInterface() {propertyValues {Curve {keys [<at 0 0>,<at 0.1 1>,<at 0.9 1>,<at 1 1>]}}})
    Scale RGB=&MakeVectorFromLinearColorRGB(Color=&LerpLinearColors(StartColor=<rgba 1 1 1 1>, EndColor=<rgba 0.1861 1.9469 2 1>, LerpFactor=$User.SlurpColorAlpha))
    ScaleA=true
    ScaleRGB=true
   )
   &GenerateDeathEvent(
    Gap Correction Amount=1.5
    Position To Send:Vector3f=&TransformPosition(Position=Linked() {linkedVariable {name=Particles.Position, type=NiagaraPosition}}, Source Space=::Local, Destination Space=::World)
   )
  ]
  rendererClasses [&NiagaraSpriteRendererProperties]
  renderers [{rendererClass=&NiagaraSpriteRendererProperties}]
 }
 Drip_Chest_B {
  emitterUpdateScript [&EmitterState(Life Cycle Mode=::NewEnumerator1, Loop Duration=1, UseLoopDelay=false)]
  particleSpawnScript [
   &InitializeParticle(
    Lifetime=&UniformRangedFloat(Minimum=0.55, Maximum=0.65)
    Mass=&UniformRangedFloat(Minimum=0.75, Maximum=1.75)
    Module.Write Color=true
    Module.Write Lifetime=true
    Module.Write Mass=false
    Module.Write Material Random=true
    Module.Write Position=false
    Module.Write Scale=false
    Module.Write SpriteRotation=false
    Module.Write SpriteSize=true
    Position=&SimulationPosition()
    Sprite Size=&Vector2DFromFloat(Value=&MakeFloatFromVector(Vector=&UniformRangedVector(Minimum=<5 0 0>, Maximum=<6 1 1>), Channel=::NewEnumerator0))
   )
   /Niagara/Modules/Spawn/Location/SphereLocation(MaskSpawnGroup=false, Module.Override Randomness=false, Offset=<-10 0 0>, Sphere Radius=0, Surface Only=false)
   &InheritVelocity(Module.Limit Speed=false, Velocity Scale=<0.3 0.3 0.3>)
  ]
  particleUpdateScript [
   &CameraOffset(Camera Offset Amount=6) {enabled=false}
   &ParticleState()
   &AccelerationForce(Acceleration=&VectorFromCurve(VectorCurve=DataInterface() {propertyValues {
       XCurve {keys [<at 0 0>,<at 0.9989 0>]}
       YCurve {keys [<at 0 0>,<at 0.9989 0>]}
       ZCurve {keys [
         {arriveTangent=-15.5802, interpMode=RCIM_Cubic, leaveTangent=-8.2205, tangentMode=RCTM_User, time=0, value=0}
         {arriveTangent=-134.1077, interpMode=RCIM_Cubic, leaveTangent=-221.0645, tangentMode=RCTM_User, time=0.3, value=-600}
        ]}
      }}))
   &SolveForcesAndVelocity(Clamp Velocity=false, Limit Acceleration=false)
   &ScaleSpriteSize(Scale Factor=&Vector2DFromCurve(
     Vector2Curve=DataInterface() {propertyValues {
       XCurve {keys [
         {arriveTangent=5.9728, interpMode=RCIM_Cubic, leaveTangent=5.9728, tangentMode=RCTM_User, time=-0.0028, value=0.5817}
         {arriveTangent=-0.1, interpMode=RCIM_Cubic, leaveTangent=-0.1, tangentMode=RCTM_User, time=0.1737, value=0.9552}
         {interpMode=RCIM_Cubic, tangentMode=RCTM_User, time=0.9956, value=0.1086}
        ]}
       YCurve {keys [
         {arriveTangent=6.2715, interpMode=RCIM_Cubic, leaveTangent=6.2715, tangentMode=RCTM_User, time=-0.0028, value=0.5817}
         {arriveTangent=2.8574, interpMode=RCIM_Cubic, leaveTangent=2.8574, tangentMode=RCTM_User, time=0.2094, value=1.5556}
         {interpMode=RCIM_Cubic, tangentMode=RCTM_User, time=0.9935, value=2.8163}
        ]}
      }}
     ScaleCurve=<1 1>
    ))
   &ScaleColor(
    Scale Alpha=0.9
    Scale RGB=&MakeVectorFromLinearColorRGB(Color=&LerpLinearColors(StartColor=<rgba 1 1 1 1>, EndColor=<rgba 0.1861 1.9469 2 1>, LerpFactor=$User.SlurpColorAlpha))
    ScaleA=true
    ScaleRGB=true
   )
  ]
  rendererClasses [&NiagaraSpriteRendererProperties]
  renderers [{rendererClass=&NiagaraSpriteRendererProperties}]
 }
 Drip_Chest_B001 {
  emitterUpdateScript [&EmitterState(Life Cycle Mode=::NewEnumerator1, Loop Duration=1, UseLoopDelay=false)]
  particleSpawnScript [
   &InitializeParticle(
    Lifetime=&UniformRangedFloat(Minimum=0.55, Maximum=0.8)
    Mass=&UniformRangedFloat(Minimum=0.75, Maximum=1.75)
    Module.Write Color=true
    Module.Write Lifetime=true
    Module.Write Mass=false
    Module.Write Material Random=true
    Module.Write Position=false
    Module.Write Scale=false
    Module.Write SpriteRotation=false
    Module.Write SpriteSize=true
    Position=&SimulationPosition()
    Sprite Size=&Vector2DFromFloat(Value=&MakeFloatFromVector(Vector=&UniformRangedVector(Minimum=<5 0 0>, Maximum=<6 1 1>), Channel=::NewEnumerator0))
   )
   &SphereLocation(MaskSpawnGroup=false, Module.Override Randomness=false, Offset=<-10 0 0>, Sphere Radius=0, Surface Only=false)
   &InheritVelocity(Module.Limit Speed=false, Velocity Scale=<0.3 0.3 0.3>)
  ]
  particleUpdateScript [
   &CameraOffset(Camera Offset Amount=6) {enabled=false}
   &ParticleState()
   &AccelerationForce(Acceleration=&VectorFromCurve(VectorCurve=DataInterface() {propertyValues {
       XCurve {keys [<at 0 0>,<at 0.9989 0>]}
       YCurve {keys [<at 0 0>,<at 0.9989 0>]}
       ZCurve {keys [
         {arriveTangent=-15.5802, interpMode=RCIM_Cubic, leaveTangent=-8.2205, tangentMode=RCTM_User, time=0, value=0}
         {arriveTangent=-134.1077, interpMode=RCIM_Cubic, leaveTangent=-221.0645, tangentMode=RCTM_User, time=0.3, value=-600}
        ]}
      }}))
   &SolveForcesAndVelocity(Clamp Velocity=false, Limit Acceleration=false)
   &ScaleSpriteSize(Scale Factor=&Vector2DFromCurve(
     Vector2Curve=DataInterface() {propertyValues {
       XCurve {keys [
         {arriveTangent=5.9728, interpMode=RCIM_Cubic, leaveTangent=5.9728, tangentMode=RCTM_User, time=0, value=0.5817}
         {arriveTangent=-0.1, interpMode=RCIM_Cubic, leaveTangent=-0.1, tangentMode=RCTM_User, time=0.173, value=0.9552}
         {interpMode=RCIM_Cubic, tangentMode=RCTM_User, time=0.9982, value=0.1086}
        ]}
       YCurve {keys [
         {arriveTangent=6.2715, interpMode=RCIM_Cubic, leaveTangent=6.2715, tangentMode=RCTM_User, time=0, value=0.5817}
         {arriveTangent=3.1201, interpMode=RCIM_Cubic, leaveTangent=3.1201, tangentMode=RCTM_User, time=0.1826, value=1.5556}
         {interpMode=RCIM_Cubic, tangentMode=RCTM_User, time=1.0091, value=2.7537}
        ]}
      }}
     ScaleCurve=<1 1>
    )) {module=ScaleSpriteSize001}
   &ScaleColor(
    Scale Alpha=0.9
    Scale RGB=&MakeVectorFromLinearColorRGB(Color=&LerpLinearColors(StartColor=<rgba 1 1 1 1>, EndColor=<rgba 0.1861 1.9469 2 1>, LerpFactor=$User.SlurpColorAlpha))
    ScaleA=true
    ScaleRGB=true
   )
  ]
  rendererClasses [&NiagaraSpriteRendererProperties]
  renderers [{rendererClass=&NiagaraSpriteRendererProperties}]
 }
}
summary {
 systemName=NS_Slurp_Drips_Bandolier
 emitters [
  {emitterName=Drip_Chest_A, rendererClasses [&NiagaraSpriteRendererProperties]}
  {emitterName=Drip_Chest_B, rendererClasses [&NiagaraSpriteRendererProperties]}
  {emitterName=Drip_Chest_A001, rendererClasses [&NiagaraSpriteRendererProperties]}
  {emitterName=Drip_Chest_B001, rendererClasses [&NiagaraSpriteRendererProperties]}
 ]
 userVariables [
  User.SkeletalMesh {
   type=NiagaraDataInterfaceSkeletalMesh
   defaultValue {
    struct=NiagaraExt_VariableValue_DataInterface
    value {dataInterface=&self:NiagaraDataInterfaceSkeletalMesh_1}
   }
  }
  User.SlurpColorAlpha {type=NiagaraFloat, defaultValue=0, description=""}
 ]
}
